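# Niagara system for the RLCS2201 vehicle boost cosmetic, drawing layered ribbon trails from the car's boost sockets.
# Two emitters spawn quality-scaled ribbons per port, with skeletal positioning and painted-variant colors, each feeding five ribbon renderers.
# Per-particle curves set layer colors and fades, while SpawnGroup-indexed arrays give signed widths that differ per ribbon.
# A frontend-preview flag switches spawn offset, backward velocity and alpha curves.
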NiagaraSystem /VehicleCosmetics/Boosts/Boost_RLCS2201/FX/NS_Boost_RLCS2201
dependencies {
 usedDataInterfaces [
  &NiagaraDataInterfaceCurve
  &NiagaraDataInterfaceArrayFloat
  &NiagaraDataInterfaceColorCurve
  &NiagaraDataInterfaceSkeletalMesh
 ]
 usedDynamicInputs []
 usedModules [
  &self:SystemSpawnScript.SystemScriptSource.SystemScriptGraph.NiagaraNodeAssignment_0.SetVariables_#1
  &SystemState
 ]
}
emitters {
 NE_VC_Boost_Ribbon_Parent {
  emitterUpdateScript [
   &EmitterState()
   &VC_FilteredSocketCount()
   &VC_RateQualityHandler(High Spawn Rate=60, Low Spawn Rate=30, Use Custom Rate?=true)
   &VC_SpawnRibbon_v1(Number of Ribbons=$Emitter.PortCount, Spawn Time=-1, SpawnRate=$Emitter.Spawn Rate)
  ]
  particleSpawnScript [
   &VC_PaintedVariantsOverride(DefaultColor=<rgba 0.1759 0.104 0.2 0.5>, Orange=<rgba 0.6724 0.2346 0 1>)
   &VC_InitializeRibbon_v1(
    Color=$Particles.Color
    Custom Offset=true
    Enable Skeletal Position=true
    Lifetime=0.7
    Local Vector Offset=/Niagara/DynamicInputs/TypeConversions/MakeCustomPositionFromBool(
     Bool:NiagaraBool=$User.bIsFrontendPreview
     True Position:NiagaraPosition=&AddVectorToPosition(Position=<-5 0 0>, Vector=$User.DTCarOffset)
     False Position:NiagaraPosition=&AddVectorToPosition(Position=<50 0 0>, Vector=$User.DTCarOffset)
    )
    Offset Amount=1
    Ribbon Width=70
   )
   &AddVelocity(Velocity=&MakeCustomVectorFromBool(Bool=$User.bIsFrontendPreview, True Vector=<-1500 0 0>, False Vector=<-700 0 0>)) {module=AddVelocity001}
  ]
  particleUpdateScript [
   &ParticleState()
   &SolveForcesAndVelocity(Clamp Velocity=false, Limit Acceleration=false)
   &VC_UpdateRibbon_v1(Enable Skeletal Position=true)
   &ScaleColor(
    Scale Alpha=&MakeCustomFloatFromBool(
     Bool=$User.bIsFrontendPreview
     True Float=&FloatFromCurve(
      FloatCurve=DataInterface() {propertyValues {Curve {keys [
          {
           arriveTangent=0.063
           arriveTangentWeight=0.1727
           interpMode=RCIM_Cubic
           leaveTangent=0.063
           leaveTangentWeight=0.1727
           tangentMode=RCTM_User
           tangentWeightMode=RCTWM_WeightedBoth
           time=0
           value=0
          }
          <at 0.499 1 RCIM_Cubic>
          {arriveTangent=-0.0405, arriveTangentWeight=0.167, interpMode=RCIM_Cubic, leaveTangent=-0.0405, tangentMode=RCTM_User, time=1, value=0}
         ]}}}
      Scale Curve=0.8
     )
     False Float=&FloatFromCurve(FloatCurve=DataInterface() {propertyValues {Curve {
         defaultValue=1
         keys [
          <at 0.499 1 RCIM_Cubic>
          {arriveTangent=-0.0405, arriveTangentWeight=0.167, interpMode=RCIM_Cubic, leaveTangent=-0.0405, tangentMode=RCTM_User, time=1, value=0}
         ]
        }}})
    )
    Scale RGB=&VectorFromFloat(Value=&FloatFromCurve(FloatCurve=DataInterface() {propertyValues {Curve {
         defaultValue=1
         keys [
          {tangentMode=RCTM_User, time=0, value=1}
          {arriveTangent=-0.0405, interpMode=RCIM_Cubic, leaveTangent=-0.0405, tangentMode=RCTM_User, time=0.7009, value=1}
          {arriveTangent=-8.0089, interpMode=RCIM_Cubic, leaveTangent=-8.0089, tangentMode=RCTM_User, time=1, value=0}
         ]
        }}}))
    ScaleA=true
    ScaleRGB=true
   )
   /SetVariables(
    Particles.Left_PurpleWidth:NiagaraFloat=&SelectFloatFromArray(Direct Array Index=$Particles.SpawnGroup, Float Selection Array=DataInterface() {propertyValues {FloatData [0,50]}})
    Particles.R1Scaled:LinearColor=&MakeLinearColorFromVectorAndFloat(
     "Vector (RGB)"=&MakeVectorFromLinearColorRGB(Color=<rgba 0.75 0.5925 0.4125 1>)
     "Float (Alpha)"=&MakeCustomFloatFromBool(
      Bool=$User.bIsFrontendPreview
      True Float=&FloatFromCurve(FloatCurve=DataInterface() {propertyValues {Curve {
          defaultValue=1
          keys [
           {
            arriveTangent=0.063
            arriveTangentWeight=0.2822
            interpMode=RCIM_Cubic
            leaveTangent=0.063
            leaveTangentWeight=0.2822
            tangentMode=RCTM_User
            tangentWeightMode=RCTWM_WeightedBoth
            time=0
            value=-0.04
           }
           {
            arriveTangent=-0.01
            arriveTangentWeight=0.2295
            interpMode=RCIM_Cubic
            leaveTangent=-0.01
            leaveTangentWeight=0.2295
            tangentMode=RCTM_User
            tangentWeightMode=RCTWM_WeightedBoth
            time=0.5
            value=1
           }
           {
            arriveTangent=0.0027
            arriveTangentWeight=0.3474
            interpMode=RCIM_Cubic
            leaveTangent=0.0027
            leaveTangentWeight=0.3474
            tangentMode=RCTM_User
            tangentWeightMode=RCTWM_WeightedBoth
            time=1
            value=-0.1
           }
          ]
         }}})
      False Float=&FloatFromCurve(FloatCurve=DataInterface() {propertyValues {Curve {keys [
           {arriveTangent=9.0725, arriveTangentWeight=0.0133, interpMode=RCIM_Cubic, leaveTangent=9.0725, tangentMode=RCTM_User, time=0.04, value=0}
           <at 0.2035 0.9847>
           {arriveTangent=-1.2363, interpMode=RCIM_Cubic, leaveTangent=-1.2363, time=0.499, value=1}
           {arriveTangent=-0.0405, arriveTangentWeight=0.167, interpMode=RCIM_Cubic, leaveTangent=-0.0405, tangentMode=RCTM_User, time=1, value=0}
          ]}}})
     )
    )
    Particles.R2Scaled:LinearColor=&MakeCustomLinearColorFromBool(
     Bool=$User.bIsFrontendPreview
     True Color=&ColorFromCurve(ColorCurve=DataInterface() {propertyValues {
        AlphaCurve {
         defaultValue=0
         keys [
          {
           arriveTangent=0.063
           arriveTangentWeight=0.2822
           interpMode=RCIM_Cubic
           leaveTangent=0.063
           leaveTangentWeight=0.2822
           tangentMode=RCTM_User
           tangentWeightMode=RCTWM_WeightedBoth
           time=0
           value=-0.04
          }
          {
           arriveTangent=-0.01
           arriveTangentWeight=0.2295
           interpMode=RCIM_Cubic
           leaveTangent=-0.01
           leaveTangentWeight=0.2295
           tangentMode=RCTM_User
           tangentWeightMode=RCTWM_WeightedBoth
           time=0.5
           value=1
          }
          {
           arriveTangent=-0.1489
           arriveTangentWeight=0.3475
           interpMode=RCIM_Cubic
           leaveTangent=-0.1489
           leaveTangentWeight=0.3475
           tangentMode=RCTM_User
           tangentWeightMode=RCTWM_WeightedBoth
           time=1
           value=-0.15
          }
         ]
        }
        BlueCurve {keys [<at 0.3419 0.8 RCIM_Cubic>,<at 0.621 0.8>,<at 0.8195 0.2681>]}
        GreenCurve {keys [<at 0.3419 0.2 RCIM_Cubic>,<at 0.621 0.2>,<at 0.8195 0.0894>]}
        RedCurve {keys [<at 0.3419 0.21 RCIM_Cubic>,<at 0.621 0.38>,<at 0.8195 0.6>]}
       }})
     False Color=&ColorFromCurve(ColorCurve=DataInterface() {propertyValues {
        AlphaCurve {
         defaultValue=1
         keys [
          {arriveTangent=9.0725, interpMode=RCIM_Cubic, leaveTangent=9.0725, tangentMode=RCTM_User, time=0.04, value=0}
          <at 0.2035 0.9847>
          {arriveTangent=-1.2363, interpMode=RCIM_Cubic, leaveTangent=-1.2363, time=0.499, value=1}
          {arriveTangent=-0.0405, arriveTangentWeight=0.167, interpMode=RCIM_Cubic, leaveTangent=-0.0405, tangentMode=RCTM_User, time=1, value=0}
         ]
        }
        BlueCurve {keys [<at 0.3419 0.8 RCIM_Cubic>,<at 0.621 0.8>,<at 0.8195 0.2681>]}
        GreenCurve {keys [<at 0.3419 0.2 RCIM_Cubic>,<at 0.621 0.2>,<at 0.8195 0.0894>]}
        RedCurve {keys [<at 0.3419 0.21 RCIM_Cubic>,<at 0.621 0.38>,<at 0.8195 0.6>]}
       }})
    )
    Particles.Right_Purple:NiagaraFloat=&SelectFloatFromArray(Direct Array Index=$Particles.SpawnGroup, Float Selection Array=DataInterface() {propertyValues {FloatData [-50,0]}})
    Particles.ThinLinesRibbonWidth:NiagaraFloat=&SelectFloatFromArray(Direct Array Index=$Particles.SpawnGroup, Float Selection Array=DataInterface() {propertyValues {FloatData [-75,75]}})
   ) {isSetParameters=true, module=SetVariables_#2}
  ]
  rendererClasses [&NiagaraRibbonRendererProperties]
  renderers [
   {rendererClass=&NiagaraRibbonRendererProperties}
   {rendererClass=&NiagaraRibbonRendererProperties, rendererIndex=1}
   {rendererClass=&NiagaraRibbonRendererProperties, rendererIndex=2}
   {rendererClass=&NiagaraRibbonRendererProperties, rendererIndex=3}
   {rendererClass=&NiagaraRibbonRendererProperties, rendererIndex=4}
  ]
 }
 NE_VC_Boost_Ribbon_Parent003 {
  emitterUpdateScript [
   &EmitterState()
   &VC_FilteredSocketCount()
   &VC_RateQualityHandler(High Spawn Rate=60, Low Spawn Rate=20, Use Custom Rate?=true)
   &VC_SpawnRibbon_v1(Number of Ribbons=$Emitter.PortCount, Spawn Time=-1, SpawnRate=$Emitter.Spawn Rate)
  ]
  particleSpawnScript [
   &VC_PaintedVariantsOverride(DefaultColor=<rgba 0.6575 0.17 0.68 1>, Orange=<rgba 0.6724 0.2346 0 1>)
   &VC_InitializeRibbon_v1(
    Color=$Particles.Color
    Custom Offset=true
    Enable Skeletal Position=true
    Lifetime=0.7
    Local Vector Offset=&MakeCustomPositionFromBool(
     Bool:NiagaraBool=$User.bIsFrontendPreview
     True Position:NiagaraPosition=&AddVectorToPosition(Position=<-5 0 0>, Vector=$User.DTCarOffset)
     False Position:NiagaraPosition=&AddVectorToPosition(Position=<50 0 0>, Vector=$User.DTCarOffset)
    )
    Offset Amount=1
    Ribbon Width=55
   )
   &AddVelocity(Velocity=&MakeCustomVectorFromBool(Bool=$User.bIsFrontendPreview, True Vector=<-1500 0 0>, False Vector=<-700 0 0>)) {module=AddVelocity001}
  ]
  particleUpdateScript [
   &ParticleState()
   &SolveForcesAndVelocity(Clamp Velocity=false, Limit Acceleration=false)
   &VC_UpdateRibbon_v1(Enable Skeletal Position=true)
   &ScaleRibbonWidth(Ribbon Width Scale=&SelectFloatFromArray(Direct Array Index=$Particles.SpawnGroup, Float Selection Array=DataInterface() {propertyValues {FloatData [-1,1]}}))
   &ScaleColor(
    Scale Alpha=&MakeCustomFloatFromBool(
     Bool=$User.bIsFrontendPreview
     True Float=&FloatFromCurve(
      FloatCurve=DataInterface() {propertyValues {Curve {
         defaultValue=1
         keys [
          {
           arriveTangent=0.063
           arriveTangentWeight=0.2822
           interpMode=RCIM_Cubic
           leaveTangent=0.063
           leaveTangentWeight=0.2822
           tangentMode=RCTM_User
           tangentWeightMode=RCTWM_WeightedBoth
           time=0
           value=-0.04
          }
          {
           arriveTangent=-0.01
           arriveTangentWeight=0.2295
           interpMode=RCIM_Cubic
           leaveTangent=-0.01
           leaveTangentWeight=0.2295
           tangentMode=RCTM_User
           tangentWeightMode=RCTWM_WeightedBoth
           time=0.5
           value=1
          }
          {
           arriveTangent=0.0027
           arriveTangentWeight=0.3474
           interpMode=RCIM_Cubic
           leaveTangent=0.0027
           leaveTangentWeight=0.3474
           tangentMode=RCTM_User
           tangentWeightMode=RCTWM_WeightedBoth
           time=1
           value=-0.15
          }
         ]
        }}}
      Scale Curve=0.8
     )
     False Float=&FloatFromCurve(
      FloatCurve=DataInterface() {propertyValues {Curve {
         defaultValue=1
         keys [
          {
           arriveTangent=9.0725
           arriveTangentWeight=0.5721
           interpMode=RCIM_Cubic
           leaveTangent=9.0725
           leaveTangentWeight=0.5721
           tangentMode=RCTM_User
           time=0.04
           value=0
          }
          <at 0.2035 0.9847>
          {arriveTangent=-1.2363, interpMode=RCIM_Cubic, leaveTangent=-1.2363, time=0.499, value=1}
          {arriveTangent=-0.0405, arriveTangentWeight=0.167, interpMode=RCIM_Cubic, leaveTangent=-0.0405, tangentMode=RCTM_User, time=1, value=0}
         ]
        }}}
      Scale Curve=1.5
     )
    )
    Scale RGB=&VectorFromFloat(Value=&FloatFromCurve(FloatCurve=DataInterface() {propertyValues {Curve {
         defaultValue=1
         keys [
          {tangentMode=RCTM_User, time=0, value=1}
          {arriveTangent=-0.0405, interpMode=RCIM_Cubic, leaveTangent=-0.0405, tangentMode=RCTM_User, time=0.7009, value=1}
          {arriveTangent=-8.0089, interpMode=RCIM_Cubic, leaveTangent=-8.0089, tangentMode=RCTM_User, time=1, value=0}
         ]
        }}}))
    ScaleA=true
    ScaleRGB=true
   )
   /SetVariables(
    Particles.BlueLeftWidth:NiagaraFloat=&SelectFloatFromArray(Direct Array Index=$Particles.SpawnGroup, Float Selection Array=DataInterface() {propertyValues {FloatData [0,55]}})
    Particles.BlueRightWidth:NiagaraFloat=&SelectFloatFromArray(Direct Array Index=$Particles.SpawnGroup, Float Selection Array=DataInterface() {propertyValues {FloatData [-55,0]}})
    Particles.R1Scaled:LinearColor=&MakeCustomLinearColorFromBool(
     Bool=$User.bIsFrontendPreview
     True Color=&ColorFromCurve(ColorCurve=DataInterface() {propertyValues {
        AlphaCurve {
         defaultValue=0
         keys [
          {
           arriveTangent=0.063
           arriveTangentWeight=0.2822
           interpMode=RCIM_Cubic
           leaveTangent=0.063
           leaveTangentWeight=0.2822
           tangentMode=RCTM_User
           tangentWeightMode=RCTWM_WeightedBoth
           time=0
           value=-0.04
          }
          {
           arriveTangent=-0.01
           arriveTangentWeight=0.2295
           interpMode=RCIM_Cubic
           leaveTangent=-0.01
           leaveTangentWeight=0.2295
           tangentMode=RCTM_User
           tangentWeightMode=RCTWM_WeightedBoth
           time=0.5
           value=1
          }
          {
           arriveTangent=0.0027
           arriveTangentWeight=0.3474
           interpMode=RCIM_Cubic
           leaveTangent=0.0027
           leaveTangentWeight=0.3474
           tangentMode=RCTM_User
           tangentWeightMode=RCTWM_WeightedBoth
           time=1
           value=-0.15
          }
         ]
        }
        BlueCurve {keys [<at 0 0.7011 RCIM_Cubic>]}
        GreenCurve {keys [<at 0 0.2524 RCIM_Cubic>]}
        RedCurve {keys [<at 0 0.6535 RCIM_Cubic>]}
       }})
     False Color=&ColorFromCurve(ColorCurve=DataInterface() {propertyValues {
        AlphaCurve {
         defaultValue=0
         keys [
          {
           arriveTangent=9.0725
           arriveTangentWeight=0.4017
           interpMode=RCIM_Cubic
           leaveTangent=9.0725
           leaveTangentWeight=0.4017
           tangentMode=RCTM_User
           time=0.04
           value=0
          }
          <at 0.2035 0.9847>
          {arriveTangent=-1.2363, interpMode=RCIM_Cubic, leaveTangent=-1.2363, time=0.499, value=1}
          {arriveTangent=-0.0405, arriveTangentWeight=0.167, interpMode=RCIM_Cubic, leaveTangent=-0.0405, tangentMode=RCTM_User, time=1, value=0}
         ]
        }
        BlueCurve {keys [<at 0 0.7011 RCIM_Cubic>]}
        GreenCurve {keys [<at 0 0.2524 RCIM_Cubic>]}
        RedCurve {keys [<at 0 0.6535 RCIM_Cubic>]}
       }})
    )
    Particles.R2Scaled:LinearColor=&MakeCustomLinearColorFromBool(
     Bool=$User.bIsFrontendPreview
     True Color=&ColorFromCurve(ColorCurve=DataInterface() {propertyValues {
        AlphaCurve {
         defaultValue=0
         keys [
          {
           arriveTangent=0.063
           arriveTangentWeight=0.2822
           interpMode=RCIM_Cubic
           leaveTangent=0.063
           leaveTangentWeight=0.2822
           tangentMode=RCTM_User
           tangentWeightMode=RCTWM_WeightedBoth
           time=0
           value=-0.04
          }
          {
           arriveTangent=-0.01
           arriveTangentWeight=0.2295
           interpMode=RCIM_Cubic
           leaveTangent=-0.01
           leaveTangentWeight=0.2295
           tangentMode=RCTM_User
           tangentWeightMode=RCTWM_WeightedBoth
           time=0.5
           value=1
          }
          {
           arriveTangent=0.0027
           arriveTangentWeight=0.3474
           interpMode=RCIM_Cubic
           leaveTangent=0.0027
           leaveTangentWeight=0.3474
           tangentMode=RCTM_User
           tangentWeightMode=RCTWM_WeightedBoth
           time=1
           value=-0.15
          }
         ]
        }
        BlueCurve {keys [<at 0 0.38 RCIM_Cubic>]}
        GreenCurve {keys [<at 0 0.0038 RCIM_Cubic>]}
        RedCurve {keys [<at 0 0.0163 RCIM_Cubic>]}
       }})
     False Color=&ColorFromCurve(ColorCurve=DataInterface() {propertyValues {
        AlphaCurve {
         defaultValue=0
         keys [
          {
           arriveTangent=9.0725
           arriveTangentWeight=0.5143
           interpMode=RCIM_Cubic
           leaveTangent=9.0725
           leaveTangentWeight=0.5143
           tangentMode=RCTM_User
           time=0.04
           value=0
          }
          <at 0.2035 0.9847>
          {arriveTangent=-1.2363, interpMode=RCIM_Cubic, leaveTangent=-1.2363, time=0.499, value=1}
          {arriveTangent=-0.0405, arriveTangentWeight=0.167, interpMode=RCIM_Cubic, leaveTangent=-0.0405, tangentMode=RCTM_User, time=1, value=0}
         ]
        }
        BlueCurve {keys [<at 0 0.38 RCIM_Cubic>]}
        GreenCurve {keys [<at 0 0.0038 RCIM_Cubic>]}
        RedCurve {keys [<at 0 0.0163 RCIM_Cubic>]}
       }})
    )
    Particles.R3Scaled:LinearColor=&MakeCustomLinearColorFromBool(
     Bool=$User.bIsFrontendPreview
     True Color=&ColorFromCurve(ColorCurve=DataInterface() {propertyValues {
        AlphaCurve {keys [
          {
           arriveTangentWeight=0.2052
           interpMode=RCIM_Cubic
           leaveTangentWeight=0.2052
           tangentMode=RCTM_User
           tangentWeightMode=RCTWM_WeightedBoth
           time=0
           value=0
          }
          <at 0.2035 2.9575>
          {arriveTangent=-3.7132, interpMode=RCIM_Cubic, leaveTangent=-3.7132, time=0.51, value=2.9728}
          {
           arriveTangent=-0.0405
           arriveTangentWeight=0.3241
           interpMode=RCIM_Cubic
           leaveTangent=-0.0405
           leaveTangentWeight=0.3241
           tangentMode=RCTM_User
           tangentWeightMode=RCTWM_WeightedBoth
           time=1
           value=0
          }
         ]}
        BlueCurve {keys [<at 0 0.1002 RCIM_Cubic>]}
        GreenCurve {keys [<at 0 0.0045 RCIM_Cubic>]}
        RedCurve {keys [<at 0 0.7 RCIM_Cubic>]}
       }})
     False Color=&ColorFromCurve(ColorCurve=DataInterface() {propertyValues {
        AlphaCurve {keys [
          {
           arriveTangent=9.0725
           arriveTangentWeight=1.4444
           interpMode=RCIM_Cubic
           leaveTangent=9.0725
           leaveTangentWeight=1.4444
           tangentMode=RCTM_User
           time=0.04
           value=0
          }
          <at 0.2035 0.9847>
          {arriveTangent=-1.2363, interpMode=RCIM_Cubic, leaveTangent=-1.2363, time=0.499, value=1}
          {arriveTangent=-0.0405, arriveTangentWeight=0.167, interpMode=RCIM_Cubic, leaveTangent=-0.0405, tangentMode=RCTM_User, time=1, value=0}
         ]}
        BlueCurve {keys [<at 0 0.1002 RCIM_Cubic>]}
        GreenCurve {keys [<at 0 0.0045 RCIM_Cubic>]}
        RedCurve {keys [<at 0 0.7 RCIM_Cubic>]}
       }})
    )
   ) {isSetParameters=true, module=SetVariables_#2}
  ]
  rendererClasses [&NiagaraRibbonRendererProperties]
  renderers [
   {rendererClass=&NiagaraRibbonRendererProperties}
   {rendererClass=&NiagaraRibbonRendererProperties, rendererIndex=1}
   {rendererClass=&NiagaraRibbonRendererProperties, rendererIndex=2}
   {rendererClass=&NiagaraRibbonRendererProperties, rendererIndex=3}
   {rendererClass=&NiagaraRibbonRendererProperties, rendererIndex=4}
  ]
 }
}
summary {
 systemName=NS_Boost_RLCS2201
 emitters [
  {emitterName=NE_VC_Boost_Ribbon_Parent, rendererClasses [&NiagaraRibbonRendererProperties]}
  {emitterName=NE_VC_Boost_Ribbon_Parent003, rendererClasses [&NiagaraRibbonRendererProperties]}
 ]
 userVariables [
  User.bIsFrontendPreview {type=NiagaraBool}
  User.bIsPainted {type=NiagaraBool}
  User.ColorVariantID {type=NiagaraInt32}
  User.DTCarOffset {type=Vector3f, description="", defaultValue {struct=Vector3f, value=<0 0 0>}}
  User.PaintedColor {type=LinearColor, description="", defaultValue {struct=LinearColor, value=<rgba 1 1 1 1>}}
  User.PaintedScalar {type=NiagaraFloat, defaultValue=1, description=""}
  User.PortCount {type=NiagaraInt32, defaultValue=2}
 ]
}
